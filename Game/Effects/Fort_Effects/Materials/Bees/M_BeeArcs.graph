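# Translucent lit mesh-particle material for bee arc effects, drawing a soft band that shifts along the mesh's V coordinate.
# Emissive comes from particle color dimmed toward grazing angles by an inverted Fresnel, passed through MF_FX_EmissiveAdjustments.
# The PanMask dynamic parameter offsets a procedural V gradient band, which is multiplied by a static fade at both UV ends.
# That mask combines with particle alpha, depth fade and view-angle fade to form opacity.
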Material /Game/Effects/Fort_Effects/Materials/Bees/M_BeeArcs
graph {
 properties {
  bEnableResponsiveAA=true
  bUsedWithMeshParticles=true
  bUsedWithNiagaraMeshParticles=true
  blendMode=BLEND_Translucent
  pixelDepthOffsetMode=PDOM_Legacy
  translucencyLightingMode=TLM_VolumetricPerVertexDirectional
  usedShadingModels=Default Lit
 }
 nodes [
  n1 = OneMinus(None=Fresnel() {exponent=2})
  n2 = ParticleColor()
  MaterialFunctionCall(Emissive=Multiply(A=@n2.RGB, B=Multiply(A=@n1) {constB=0.8})) {materialFunction=&MF_FX_EmissiveAdjustments}
  n3 = DynamicParameter() {defaultValue=<rgba 0 0 0 1>, paramNames [PanMask,Param2,Param3,Param4]}
  n4 = ComponentMask(None=Add(A=TextureCoordinate(), B=@n3.PanMask)) {g=true}
  n5 = TextureCoordinate()
  n6 = OneMinus(None=Add(A=ComponentMask(None=@n5) {g=true}, B=@n3.PanMask))
  n7 = TextureCoordinate()
  n8 = ComponentMask(None=TextureCoordinate()) {g=true}
  n9 = Multiply(
   A=LinearInterpolate(Alpha=Multiply(A=@n4, B=@n6)) {constA=-0.25, constB=2.5}
   B=Multiply(A=ComponentMask(None=@n7) {g=true}, B=OneMinus(None=@n8))
  )
  Multiply(
   A=DepthFade(Opacity=Multiply(A=@n2.A, B=@n9))
   B=MaterialFunctionCall("FadeSensitivity (optional)"=Constant() {r=0.5}) {materialFunction=&FadeBasedOnViewAngle}.Result
  )
 ]
}
