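# M_Bees_Body is a two-sided Default Lit material for bee bodies in effects, enabled for mesh particles and Nanite.
# It multiplies a bee color texture by particle color, so the emitter can tint each bee.
# A bee normal map and scalar parameters for roughness, specular and metallic supply the remaining surface inputs.
# Responsive anti-aliasing is enabled.
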Material /Game/Effects/Fort_Effects/Materials/Bees/M_Bees_Body
graph {
 properties {
  bEnableResponsiveAA=true
  bUsedWithMeshParticles=true
  bUsedWithNanite=true
  pixelDepthOffsetMode=PDOM_Legacy
  twoSided=true
  usedShadingModels=Default Lit
 }
 nodes [
  Multiply(A=TextureSample() {texture=/Game/Effects/Fort_Effects/Textures/Bugs/Bees/T_Bee_Color}.RGB, B=ParticleColor().RGB)
  ScalarParameter() {defaultValue=0.4, parameterName=Roughness}
  ScalarParameter() {defaultValue=0.6, parameterName=Specular}
  ScalarParameter() {defaultValue=0.5, parameterName=Metallic}
  TextureSample() {samplerType=SAMPLERTYPE_Normal, texture=/Game/Effects/Fort_Effects/Textures/Bugs/Bees/T_Bee_Normal}
 ]
}
